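# Translucent particle material for sprite and Niagara effects, a dark fog puff in the Vampire cosmetic effect set.
# Bat glider noise textures pan on doubled dynamic-parameter Time, with UVs distorted by a sphere normal map.
# A sphere mask shapes two channels of one panned sample, which are screen-blended and scaled by particle alpha.
# Depth fade softens intersections on the default shading path; mobile skips it.
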Material /BRCosmetics/Athena/Effects/Materials/Vampire/M_DarkFog
graph {
 properties {
  bUsedWithNiagaraSprites=true
  bUsedWithParticleSprites=true
  blendMode=BLEND_Translucent
  pixelDepthOffsetMode=PDOM_Legacy
  usedShadingModels=Default Lit
 }
 nodes [
  n1 = Multiply(A=TextureCoordinate(), B=ScalarParameter() {defaultValue=1, parameterName=TestureScale})
  n2 = DynamicParameter() {paramNames [Time,Random,Param3,Param4]}
  n3 = Add(A=@n2.Time, B=@n2.Time)
  n4 = Multiply(A=ComponentMask(None=TextureSample() {samplerType=SAMPLERTYPE_Normal, texture=/Game/Environments/World/Backgrounds/Universal/TechContent/Textures/T_Spherenormal_01}.RGB) {g=true, r=true}) {constB=0.15}
  n5 = TextureSample(UVs=Add(A=Panner(Coordinate=@n1, Time=@n3) {speedX=-0.25, speedY=-0.25}, B=@n4)) {samplerType=SAMPLERTYPE_Masks, texture=/Game/Gadgets/Assets/VinderTech_GliderChute/Glider_Vampire/Textures/T_Glider_Bat_Noises}
  n6 = TextureSample(UVs=Add(A=Panner(Coordinate=@n1, Time=@n3) {speedX=0.25, speedY=0.25}, B=@n4)) {samplerType=SAMPLERTYPE_Masks, texture=&T_Glider_Bat_Noises}
  Add(A=@n5.B, B=@n6.B)
  Multiply(A=@n5.B, B=@n6.B)
  n7 = Constant() {r=0.5}
  n8 = SphereMask(A=TextureCoordinate(), B=@n7, Radius=@n7, Hardness=Constant())
  n9 = MaterialFunctionCall!(Blend=Subtract(A=@n6.B, B=OneMinus(None=@n8)), Base=Multiply(A=@n6.G, B=Multiply(B=@n8) {constA=1.5})) {materialFunction=/Engine/Functions/Engine_MaterialFunctions03/Blends/Blend_Screen}
  n10 = ParticleColor()
  n11 = Multiply(A=@n9.Result, B=@n10.A)
  Multiply(A=@n11, B=@n10.RGB)
  n12 = ScalarParameter() {defaultValue=20, parameterName=Fade Distance}
  ShadingPathSwitch(Default=Multiply(A=DepthFade(FadeDistance=@n12) {fadeDistanceDefault=20}, B=@n11), Mobile=@n11) {desc=Migrated from FeatureLevelSwitch}
  SquareRoot(None=Saturate(None=@n9.Result))
  TextureCoordinate()
  TextureCoordinate()
 ]
}
